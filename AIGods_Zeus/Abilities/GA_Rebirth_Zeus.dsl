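; Zeus's rebirth ability, extending a parent rebirth ability with Zeus-specific effects at key phases.
; When pre-rebirth starts, it spawns the thunderstorm class, if valid, at the avatar's transform and tries to activate shield abilities by tag on the AI pawn.
; Landing applies an effect preventing lightning bolt, and ability end applies one preventing lightning jump.
; Each override then calls its parent implementation.
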(event Ability|EventOnEndAbility
  (params (bWasCancelled :type Boolean))
  (Ability|ApplyGameplayEffectToOwner "/AIGods_Zeus/Abilities/GE_Zeus_LightningJump_Prevent.GE_Zeus_LightningJump_Prevent_C" 0)
  (|CallParentFunction bWasCancelled))

(event EventStartPreRebirth
  (bind _thunderstormclass (Variables|Default|GetThunderStormClass))
  (if (Utilities|IsValidClass _thunderstormclass)
    (Game|SpawnActorfromClass _thunderstormclass (Transformation|GetActorTransform (Ability|GetAvatarActorfromActorInfo)) "AlwaysSpawn" "MultiplyWithRoot" 0 0)
    (else
      _))
  (Abilities|TryActivateAbilitiesbyTag (Ability|GetAbilitySystemComponent (Variables|InternalData|GetAIPawn)) (Variables|Default|GetTagContainer_Shield) false)
  (|CallParentFunction))

(fn OnPawnLanded
  (Ability|ApplyGameplayEffectToOwner (Variables|Default|GetGE_PreventLightningBolt) 0)
  (|CallParentFunction))
